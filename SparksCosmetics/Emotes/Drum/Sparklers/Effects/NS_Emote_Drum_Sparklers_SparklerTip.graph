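# Niagara system for the burning tip of a sparkler in the drum sparklers emote, with randomly colored sparks.
# Core flicker sprites and BouncingSparks surround a persistent, color-shifting Light, while a ribbon Trail fades from warm to random color.
# GPU Streaks use two spawn groups to mix short and long gravity-affected sparks with different color, lifetime and size.
# User.Intensity drives most spawn rates and brightness, with front-end preview and time-of-day switches adjusting color and glow.
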NiagaraSystem /SparksCosmetics/Emotes/Drum/Sparklers/Effects/NS_Emote_Drum_Sparklers_SparklerTip
dependencies {
 usedDataInterfaces [
  &NiagaraDataInterfaceCurve
  &NiagaraDataInterfaceVector2DCurve
  &NiagaraDataInterfaceColorCurve
  &NiagaraDataInterfaceSpriteRendererInfo
  &NiagaraDataInterfaceVectorCurve
 ]
 usedDynamicInputs [&Multiply_Vector2D_ByFloat]
 usedModules [&SystemState]
}
emitters {
 BouncingSparks {
  emitterUpdateScript [
   &EmitterState()
   &SpawnRate(SpawnRate=&Multiply_Float(A=100, B=$User.Intensity), Use Spawn Probability=false)
  ]
  particleSpawnScript [
   &InitializeParticle(
    Color Channel Mode:ENiagaraChannelCorrelation=::NewEnumerator2
    Color Maximum=<rgba 1 1 1 0.25>
    Color Minimum=<rgba 0 0 0 0.25>
    Color Mode=::NewEnumerator2
    Lifetime=&RandomRangeFloat(Minimum=0.025, Maximum=0.1, Evaluation Type=::NewEnumerator1)
    Sprite Rotation Mode=::NewEnumerator2
    Sprite Size Mode=::NewEnumerator1
    Uniform Sprite Size Min=2
    UsePositionOffset=false
    Write Material Random=true
   )
   &ShapeLocation(
    MaskSpawnGroup=false
    Override Seed=false
    Sphere Radius=&RandomRangeFloat(Minimum=5, Maximum=10, Evaluation Type=::NewEnumerator1)
    Sphere Surface Distribution=1
    UseOwnerScale=true
   )
   /CascadeToNiagaraConverter/NiagaraScripts/ModuleScripts/FixShapeVector()
   &AddVelocity(Module.Use Falloff Distance=false, Override Seed=false, Velocity Mode=::NewEnumerator1)
  ]
  particleUpdateScript [
   &ParticleState()
   &ScaleSpriteSize(
    Scale Factor=&Vector2DFromFloat(Value=&FloatFromCurve(FloatCurve=DataInterface() {propertyValues {Curve {keys [<at 0 0>,<at 1 1>]}}}))
    Scale Sprite Size Mode=::NewEnumerator2
   )
   &SubUVAnimation(
    End Frame Range Override=2
    "Sprite Renderer "=DataInterface() {propertyValues {SpriteRenderer=&self:BouncingSparks_0.NiagaraSpriteRendererProperties_2}}
    UseEndFrame=false
    UseStartFrame=false
   )
   &Drag(Drag=3, Use Linear Drag=true, Use Rotational Drag=true)
   &SolveForcesAndVelocity(Clamp Velocity=false, Limit Acceleration=false)
   &ScaleColor(
    Scale RGB=&MakeVectorFromLinearColorRGB(Color=&Multiply_LinearColorByFloat(
      Linear Color=&MakeCustomLinearColorFromBool(
       Bool=$User.bIsFrontEndPreview
       True Color=$Particles.Initial.Color
       False Color=/Niagara/DynamicInputs/Add/Add_LinearColor(
        A:LinearColor=/Niagara/DynamicInputs/TypeConversions/MakeLinearColorFromFloat(Value:NiagaraFloat=0.5)
        B:LinearColor=&Multiply_LinearColorByFloat(Linear Color=$Particles.Initial.Color, Float=0.5)
       )
      )
      Float=&Lerp_Float(A=1, B=&RandomRangeFloat(Minimum=1.25, Maximum=2), Alpha=$User.Intensity)
     ))
    ScaleA=false
    ScaleRGB=true
   )
  ]
  rendererClasses [&NiagaraSpriteRendererProperties]
  renderers [{rendererClass=&NiagaraSpriteRendererProperties}]
 }
 Core {
  emitterUpdateScript [
   &EmitterState()
   &SpawnRate(SpawnRate=100, Use Spawn Probability=false)
  ]
  particleSpawnScript [&InitializeParticle(
    Color Channel Mode:ENiagaraChannelCorrelation=::NewEnumerator2
    Color Minimum=<rgba 0.1 0.1 0.1 1>
    Color Mode=::NewEnumerator2
    Lifetime=&RandomRangeFloat(Minimum=0.025, Maximum=0.1, Evaluation Type=::NewEnumerator1)
    Sprite Rotation Mode=::NewEnumerator2
    Sprite Size=&Vector2DFromFloat(Value=&RandomRangeFloat(Minimum=5, Maximum=20, Evaluation Type=::NewEnumerator1))
    Sprite Size Mode=::NewEnumerator4
    UsePositionOffset=false
    Write Material Random=true
   )]
  particleUpdateScript [
   &ParticleState()
   &ScaleSpriteSize(
    Scale Sprite Size Mode=::NewEnumerator0
    Uniform Scale Factor=&FloatFromCurve(FloatCurve=DataInterface() {propertyValues {Curve {keys [<at 0 0>,<at 1 1>]}}})
   )
   &SubUVAnimation(
    End Frame Range Override=2
    "Sprite Renderer "=DataInterface() {propertyValues {SpriteRenderer=&self:Core_0.NiagaraSpriteRendererProperties_3}}
    UseEndFrame=false
    UseStartFrame=false
   )
   &CameraOffset(Camera Offset Amount=2)
   &ScaleColor(
    Scale RGB=&VectorFromFloat(Value=&MakeCustomFloatFromBool(
      Bool=$User.bIsFrontEndPreview
      True Float=1
      False Float=&Multiply_Float(
       A=&FNTimeOfDayScalar(Night Scale=0.75, Morning Scale=0.5, Evening Scale=0.5)
       B=&Lerp_Float(A=1, B=&RandomRangeFloat(Minimum=1.25, Maximum=2), Alpha=$User.Intensity)
      )
     ))
    ScaleA=true
    ScaleRGB=true
   )
  ]
  rendererClasses [&NiagaraSpriteRendererProperties]
  renderers [{rendererClass=&NiagaraSpriteRendererProperties}]
 }
 Light {
  emitterUpdateScript [
   &EmitterState()
   &SpawnBurst_Instantaneous(Use Spawn Probability=false, UseLoopCountLimit=true)
  ]
  particleSpawnScript [
   &InitializeParticle(
    Color Mode=::NewEnumerator0
    Lifetime=5
    Position Offset=<2 0 0>
    Position Offset Coordinate Space=::Local
    Sprite Size Mode=::NewEnumerator3
    Uniform Sprite Size=3
    UsePositionOffset=true
    Write Material Random=true
   )
   /SetVariables(Particles.LightExponent=2) {isSetParameters=true}
  ]
  particleUpdateScript [
   &ParticleState(Kill Particles When Lifetime Has Elapsed=false, Let Infinitely Lived Particles Die When Emitter Deactivates=true)
   &InterpolateOverTime(
    Initial Color Value:LinearColor=<rgba 1 1 1 1>
    Rate Of Change=7
    Target Color Value:LinearColor=/Niagara/DynamicInputs/LinearColor/MakeLinearColorFromHSV(
     HSV From:NiagaraInt32=1
     HSV:Vector3f=?
     Hue:NiagaraFloat=&RandomRangeFloat(Minimum=0, Maximum=1, Evaluation Type=::NewEnumerator1)
     Saturation:NiagaraFloat=1.2
     Value:NiagaraFloat=&RandomRangeFloat(Minimum=3, Maximum=10, Evaluation Type=::NewEnumerator1)
     Alpha:NiagaraFloat=1
    )
    Variable Type=::NewEnumerator4
   )
   &InterpolateOverTime(
    Initial Float Value=50
    Rate Of Change=10
    Target Float Value:NiagaraFloat=&RandomRangeFloat(Minimum=40, Maximum=80, Evaluation Type=::NewEnumerator1)
   ) {module=InterpolateOverTime001}
  ]
  rendererClasses [&NiagaraSpriteRendererProperties, &NiagaraLightRendererProperties]
  renderers [
   {rendererClass=&NiagaraSpriteRendererProperties}
   {rendererClass=&NiagaraLightRendererProperties, rendererIndex=1}
  ]
 }
 Streaks {
  simTarget=GPUComputeSim
  emitterUpdateScript [
   &EmitterState()
   &SpawnRate(SpawnRate=&Multiply_Float(A=600, B=$User.Intensity), Use Spawn Probability=false)
   &SpawnRate(SpawnGroup=1, SpawnRate=&Multiply_Float(A=100, B=$User.Intensity), Use Spawn Probability=false) {module=SpawnRate001}
  ]
  particleSpawnScript [
   /SetVariables(
    Transient.ShortColor:LinearColor=&MakeLinearColorFromHSV(
     HSV From:NiagaraInt32=1
     HSV:Vector3f=?
     Hue:NiagaraFloat=&RandomRangeFloat(Minimum=0, Maximum=1)
     Saturation:NiagaraFloat=&FN_FrontEndFloatModifier(
      Gameplay=&FNTimeOfDayScalar(
       Night Scale=&RandomRangeFloat(Minimum=0.75, Maximum=0.85)
       Morning Scale=&RandomRangeFloat(Minimum=0.75, Maximum=0.85)
       Day Scale=&RandomRangeFloat(Minimum=1, Maximum=2)
       Evening Scale=&RandomRangeFloat(Minimum=0.75, Maximum=0.85)
      )
      Front End=&RandomRangeFloat(Minimum=0.75, Maximum=0.85)
     )
     Value:NiagaraFloat=1
     Alpha:NiagaraFloat=1
    )
    Transient.ShortLifetime:NiagaraFloat=&RandomRangeFloat(Minimum=0.025, Maximum=0.05, Evaluation Type=::NewEnumerator1)
    Transient.ShortScale:Vector2f=&RandomRangeVector2D(Minimum=<1 3>, Maximum=<1.5 4>, Evaluation Type=::NewEnumerator1)
   ) {isSetParameters=true}
   /SetVariables(
    Transient.LongColor:LinearColor=&RandomRangeLinearColor(Minimum=<rgba 0.1 0.1 0.1 1>, Maximum=<rgba 1 1 1 1>, Color Channel Mode=::NewEnumerator2)
    Transient.LongLifetime:NiagaraFloat=&RandomRangeFloat(Minimum=0.1, Maximum=0.5, Evaluation Type=::NewEnumerator1)
    Transient.LongScale:Vector2f=&RandomRangeVector2D(Minimum=<1 1.5>, Maximum=<1.5 1.5>, Evaluation Type=::NewEnumerator1)
   ) {isSetParameters=true}
   /SetVariables(Particles.VisibilityTag=&SpawnGroup()) {isSetParameters=true}
   &InitializeParticle(
    Color=&MaskLinearColorBySpawnGroup(Masked Linear Color=$Transient.ShortColor, Passthrough Linear Color=$Transient.LongColor)
    Lifetime=&MaskFloatBySpawnGroup(Masked Float=$Transient.ShortLifetime, Passthrough Float=$Transient.LongLifetime)
    Sprite Size=&MaskVector2dBySpawnGroup(Masked Vector2D=$Transient.ShortScale, Passthrough Vector2D=$Transient.LongScale, Spawn Group Mask=0)
    Sprite Size Mode=::NewEnumerator4
    UsePositionOffset=false
    Write Material Random=true
   )
   &AddVelocity(Velocity=&RandomRangeVector(Minimum=<-250 -250 -250>, Maximum=<250 250 250>, Evaluation Type=::NewEnumerator1))
  ]
  particleUpdateScript [
   &ParticleState()
   &ScaleSpriteSize(
    Scale Factor=&Vector2DFromCurve(
     Vector2Curve=DataInterface() {propertyValues {
       XCurve {keys [<at 0 1 RCIM_Cubic>, {arriveTangent=-1.9252, interpMode=RCIM_Cubic, leaveTangent=-1.9252, tangentMode=RCTM_User, time=1, value=0.2}]}
       YCurve {keys [<at 0 3 RCIM_Cubic>, {interpMode=RCIM_Cubic, tangentMode=RCTM_User, time=1, value=0}]}
      }}
     ScaleCurve=<1 1>
    )
    Scale Sprite Size Mode=::NewEnumerator2
   )
   &AccelerationForce(Acceleration=<0 0 -980>)
   &Drag(Use Linear Drag=true, Use Rotational Drag=true)
   &SolveForcesAndVelocity(Clamp Velocity=false, Limit Acceleration=false)
   /SetVariables(
    Transient.LongColorVector:Vector3f=&MakeVectorFromLinearColorRGB(Color=&MultiplyLinearColor(
      Linear Color=&ColorFromCurve(ColorCurve=DataInterface() {propertyValues {
         AlphaCurve {keys [<at 0 1>]}
         BlueCurve {keys [<at 0.15 0.2338>,<at 0.3 1>]}
         GreenCurve {keys [<at 0.15 0.7554>,<at 0.3 1>]}
         RedCurve {keys [<at 0.15 1.25>,<at 0.3 1>]}
        }})
      Scale Factor=/Niagara/DynamicInputs/TypeConversions/MakeCustomVector4FromBool(
       Bool:NiagaraBool=$User.bIsFrontEndPreview
       True Vector4:Vector4f=&MakeVector4_FromLinearColor(LinearColor=$Particles.Initial.Color)
       False Vector4:Vector4f=&MakeVector4_FromFloat(Float=1)
      )
     ))
    Transient.ShortColorVector:Vector3f=&VectorFromFloat(Value=&RandomRangeFloat(Minimum=1, Maximum=1.5))
   ) {isSetParameters=true}
   &ScaleColor(
    Scale Alpha=&FNTimeOfDayScalar(Night Scale=0.5, Morning Scale=0.5, Evening Scale=0.5)
    Scale RGB=&MaskVectorBySpawnGroup(Masked Vector=$Transient.ShortColorVector, Passthrough Vector=$Transient.LongColorVector)
    ScaleA=true
    ScaleRGB=true
   )
   /SetVariables(Particles.Position=/Niagara/DynamicInputs/Helpers/MaskPosition(
     Mask Source:NiagaraInt32=0
     Mask Parameter:NiagaraInt32=?
     Mask:NiagaraInt32=0
     Masked Position:NiagaraPosition=$Engine.Owner.Position
     Passthrough Position:NiagaraPosition=$Particles.Position
    )) {isSetParameters=true}
  ]
  rendererClasses [&NiagaraSpriteRendererProperties]
  renderers [
   {rendererClass=&NiagaraSpriteRendererProperties}
   {rendererClass=&NiagaraSpriteRendererProperties, rendererIndex=1}
  ]
 }
 Trail {
  emitterUpdateScript [
   &EmitterState()
   &SpawnRate(SpawnRate=100, Use Spawn Probability=false)
  ]
  particleSpawnScript [
   /SetVariables(Particles.Acceleration:Vector3f=&RandomRangeVector(Minimum=<-3 -3 0>, Maximum=<3 3 0>, Evaluation Type=::NewEnumerator1)) {isSetParameters=true}
   &InitializeParticle(
    Color=<rgba 10 5.7477 1.89 1>
    Lifetime=0.2
    Ribbon Width=&Multiply_Float(A=0.75, B=2)
    Ribbon Width Mode=::NewEnumerator1
    UsePositionOffset=false
    Write Material Random=true
   )
  ]
  particleUpdateScript [
   /SetVariables(Particles.ScaledRibbonWidth:NiagaraFloat=&Multiply_Float(
     A=$Particles.RibbonWidth
     B=&FloatFromCurve(FloatCurve=DataInterface() {propertyValues {Curve {keys [<at 0 2>,<at 0.2 1>,<at 1 0>]}}})
    )) {isSetParameters=true}
   &ParticleState()
   &AccelerationForce(Acceleration=$Particles.Acceleration)
   &SolveForcesAndVelocity(Clamp Velocity=false, Limit Acceleration=false)
   &DynamicMaterialParameters(
    Index 0 Param 1=0.5
    Index 0 Param 2=0.25
    Index 0 Param 3=1
    Index 0 Param 4=1
    Module.Param0WriteEnabled=true
    Module.Param1WriteEnabled=true
    Module.Param2WriteEnabled=true
    Module.Param3WriteEnabled=true
   )
   &ScaleColor(
    Scale Alpha=&FloatFromCurve(
     FloatCurve=DataInterface() {propertyValues {Curve {keys [{arriveTangent=-2.4627, interpMode=RCIM_Cubic, leaveTangent=-2.4627, tangentMode=RCTM_User, time=0, value=1}, <at 1 0>]}}}
     Scale Curve=$User.Intensity
    )
    Scale RGB=&VectorFromCurve(VectorCurve=DataInterface() {propertyValues {
       XCurve {keys [{arriveTangent=-7.3657, interpMode=RCIM_Cubic, leaveTangent=-7.3657, tangentMode=RCTM_User, time=0, value=5}, <at 1 1>]}
       YCurve {keys [{arriveTangent=-3.4554, interpMode=RCIM_Cubic, leaveTangent=-3.4554, tangentMode=RCTM_User, time=0, value=3.1353}, <at 1 1>]}
       ZCurve {keys [{arriveTangent=-1.5714, interpMode=RCIM_Cubic, leaveTangent=-1.5714, tangentMode=RCTM_User, time=0, value=1.8534}, <at 1 1>]}
      }})
    ScaleA=true
    ScaleRGB=true
   ) {enabled=false}
   &Color(Color=&self:PowerColor(
     Color:LinearColor=&MakeLinearColorFromVectorAndFloat(
      "Vector (RGB)"=&MakeVectorFromLinearColorRGB(Color=&LerpLinearColors(
        StartColor=$Particles.Initial.Color
        EndColor=&Multiply_LinearColorByFloat(
         Linear Color=&RandomRangeLinearColor(Minimum=<rgba 0 0 0 1>, Maximum=<rgba 1 1 1 1>, Color Channel Mode=::NewEnumerator2)
         Float=5
        )
        LerpFactor=&FloatFromCurve(FloatCurve=DataInterface() {propertyValues {Curve {keys [{arriveTangent=3.742, interpMode=RCIM_Cubic, leaveTangent=3.742, tangentMode=RCTM_User, time=0, value=0}, <at 0.5 0.95>]}}})
       ))
      "Float (Alpha)"=&FloatFromCurve(
       FloatCurve=DataInterface() {propertyValues {Curve {keys [{arriveTangent=-2.4627, interpMode=RCIM_Cubic, leaveTangent=-2.4627, tangentMode=RCTM_User, time=0, value=1}, <at 1 0>]}}}
       Scale Curve=&Multiply_Float(
        A=&Multiply_Float(A=$User.Intensity, B=&FN_FrontEndFloatModifier(Gameplay=1, Front End=0.1))
        B=&MakeFloatFromLinearColor(LinearColor=$Particles.Initial.Color, Channel:ENiagaraFloatFromLinearColorOptions=::NewEnumerator3)
       )
      )
     )
     Power:NiagaraFloat=&FN_FrontEndFloatModifier(Gameplay=1, Front End=2)
    ))
  ]
  rendererClasses [&NiagaraRibbonRendererProperties]
  renderers [{rendererClass=&NiagaraRibbonRendererProperties}]
 }
}
summary {
 systemName=NS_Emote_Drum_Sparklers_SparklerTip
 emitters [
  {emitterName=Streaks, simTarget=GPUComputeSim, rendererClasses [&NiagaraSpriteRendererProperties]}
  {emitterName=BouncingSparks, rendererClasses [&NiagaraSpriteRendererProperties]}
  {emitterName=Core, rendererClasses [&NiagaraSpriteRendererProperties]}
  {emitterName=Trail, rendererClasses [&NiagaraRibbonRendererProperties]}
  {emitterName=Light, rendererClasses [&NiagaraSpriteRendererProperties, &NiagaraLightRendererProperties]}
 ]
 userVariables [
  User.bIsFrontEnd {type=NiagaraBool}
  User.bIsFrontEndPreview {type=NiagaraBool}
  User.Intensity {type=NiagaraFloat, defaultValue=1, description=""}
 ]
}
