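# Niagara system for the twirl in the Sparklers drum emote, drawing short-lived trails behind two skeletal mesh sockets or bones.
# One Trail emitter runs two spawn groups, each forming its own ribbon faced by sampled rotation, alongside a sprite renderer.
# Each group's first particle is re-pinned to its sampled point every update.
# Trails narrow and fade with age, tint by time of day, and lower alpha in front-end preview.
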NiagaraSystem /SparksCosmetics/Emotes/Drum/Sparklers/Effects/NS_Emote_Drum_Sparklers_Twirl
dependencies {
 usedDataInterfaces [&NiagaraDataInterfaceCurve, &NiagaraDataInterfaceSkeletalMesh]
 usedDynamicInputs []
 usedModules [
  &self:SystemSpawnScript.SystemScriptSource.SystemScriptGraph.NiagaraNodeAssignment_0.SetVariables_#1
  &SystemState
 ]
}
emitters {Trail {
  emitterUpdateScript [
   &EmitterState()
   &SpawnRate(SpawnRate=30, Use Spawn Probability=false)
   &SpawnRate(SpawnGroup=1, SpawnRate=30, Use Spawn Probability=false) {module=SpawnRate001}
   &self:SetHasSpawned(Has Spawned 00:NiagaraBool=$Output.SpawnRate.HasSpawnedThisFrame, Has Spawned 01:NiagaraBool=$Output.SpawnRate001.HasSpawnedThisFrame) {module=SetHasSpawned}
  ]
  particleSpawnScript [
   &InitializeParticle(
    Color=<rgba 0.7156 0.7156 0.7156 0.1>
    Lifetime=0.25
    Ribbon Width=30
    Ribbon Width Mode=::NewEnumerator1
    Sprite Size Mode=::NewEnumerator3
    UsePositionOffset=false
    Write Material Random=true
   )
   &SkeletalMeshLocation(
    Bone / Socket Index=2
    Bone Sampling Mode=::NewEnumerator4
    MaskSpawnGroup=false
    Orientation Sampling=::NewEnumerator1
    Position Sampling=::NewEnumerator2
    Sampled Position Offset=<0 0 20>
    Skeletal Mesh=$System.Skeletal Mesh
    Use Sampled Mesh Attribute as Ribbon ID=false
   )
   &SkeletalMeshLocation(
    Bone / Socket Index=3
    Bone Sampling Mode=::NewEnumerator4
    MaskSpawnGroup=false
    Orientation Sampling=::NewEnumerator1
    Position Sampling=::NewEnumerator2
    Sampled Position Offset=<0 0 20>
    Skeletal Mesh=$System.Skeletal Mesh
    Use Sampled Mesh Attribute as Ribbon ID=false
   ) {module=SkeletalMeshLocation001}
   /SetVariables(
    Particles.RibbonFacing=&RotateVectorByQuaternion(
     Vector To Rotate=<0 0 5>
     Quaternion=/Niagara/DynamicInputs/Helpers/MaskQuaternion(
      Mask Source:NiagaraInt32=0
      Mask Parameter:NiagaraInt32=?
      Mask:NiagaraInt32=0
      Masked Quaternion:Quat4f=$Particles.SkeletalMeshLocation.SampledRotation
      Passthrough Quaternion:Quat4f=$Particles.SkeletalMeshLocation001.SampledRotation
     )
    )
    Particles.RibbonID=/Niagara/DynamicInputs/Meshes/MakeRibbonIDFromSpawnGroup()
   ) {isSetParameters=true}
   &AddVelocity(Velocity=<500 0 0>) {enabled=false}
  ]
  particleUpdateScript [
   &SolveForcesAndVelocity(Clamp Velocity=false, Limit Acceleration=false) {enabled=false}
   &ParticleState()
   &ScaleRibbonWidth(Ribbon Width Scale=&FloatFromCurve(FloatCurve=DataInterface() {propertyValues {Curve {keys [<at 0 1 RCIM_Cubic>, {arriveTangent=-2.073, tangentMode=RCTM_User, time=1, value=0}]}}}))
   &ScaleColor(
    Scale Alpha=&FloatFromCurve(
     FloatCurve=DataInterface() {propertyValues {Curve {keys [
         {arriveTangent=-4.2163, interpMode=RCIM_Cubic, leaveTangent=-4.2163, tangentMode=RCTM_User, time=0, value=1}
         {arriveTangent=-0.3042, interpMode=RCIM_Cubic, leaveTangent=-0.3042, tangentMode=RCTM_User, time=0.3, value=0.5}
         {arriveTangent=-1.1367, tangentMode=RCTM_User, time=1, value=0}
        ]}}}
     Scale Curve=&MakeCustomFloatFromBool(Bool=$User.bIsFrontEndPreview, True Float=0.4, False Float=1)
    )
    Scale RGB=/Niagara/DynamicInputs/Math/LerpMultipleVectors(
     Alpha:NiagaraFloat=Linked() {linkedVariable {name=NPC.FortniteNPC.TimeOfDay_0-1, type=NiagaraFloat}}
     V 1:Vector3f=<0.5 0.5 1>
     V 2:Vector3f=<1 0.75 0.5>
     V 3:Vector3f=<1 1 1>
     V 4:Vector3f=<1 0.75 0.5>
     V 5:Vector3f=<0.5 0.5 1>
     Vector Count:ENiagara_MultipleLerpCount=::NewEnumerator2
    )
    ScaleA=true
    ScaleRGB=true
   )
   &SkeletalMeshLocation(
    Bone / Socket Index=2
    Bone Sampling Mode=::NewEnumerator4
    MaskSpawnGroup=false
    Orientation Sampling=::NewEnumerator1
    Position Sampling=::NewEnumerator2
    Sampled Position Offset=<0 0 25>
    Skeletal Mesh=$System.Skeletal Mesh
    Use Sampled Mesh Attribute as Ribbon ID=false
   ) {module=SkeletalMeshLocation002}
   &SkeletalMeshLocation(
    Bone / Socket Index=3
    Bone Sampling Mode=::NewEnumerator4
    MaskSpawnGroup=false
    Orientation Sampling=::NewEnumerator1
    Position Sampling=::NewEnumerator2
    Sampled Position Offset=<0 0 25>
    Skeletal Mesh=$System.Skeletal Mesh
    Use Sampled Mesh Attribute as Ribbon ID=false
   ) {module=SkeletalMeshLocation003}
   &self:IsFirstParticle(
    Locked Color:LinearColor=&MakeLinearColorFromVectorAndFloat(
     "Vector (RGB)"=&MakeVectorFromLinearColorRGB(Color=$Particles.Color)
     "Float (Alpha)"=&MakeCustomFloatFromBool(Bool=$User.bIsFrontEndPreview, True Float=0.4, False Float=1)
    )
    Locked Position:NiagaraPosition=&MaskVectorBySpawnGroup(
     Masked Vector=Linked() {linkedVariable {name=Particles.SkeletalMeshLocation002.SampledPosition, type=NiagaraPosition}}
     Passthrough Vector=Linked() {linkedVariable {name=Particles.SkeletalMeshLocation003.SampledPosition, type=NiagaraPosition}}
    )
    Set Color=true
    Set Position=true
    Spawn Group:NiagaraInt32=&SpawnGroup()
   ) {module=IsFirstParticle}
  ]
  rendererClasses [&NiagaraSpriteRendererProperties, &NiagaraRibbonRendererProperties]
  renderers [
   {rendererClass=&NiagaraSpriteRendererProperties}
   {rendererClass=&NiagaraRibbonRendererProperties, rendererIndex=1}
  ]
 }}
summary {
 systemName=NS_Emote_Drum_Sparklers_Twirl
 emitters [{emitterName=Trail, rendererClasses [&NiagaraSpriteRendererProperties, &NiagaraRibbonRendererProperties]}]
 userVariables [
  User.bIsFrontEnd {type=NiagaraBool}
  User.bIsFrontEndPreview {type=NiagaraBool}
  User.IsFrontEnd {type=NiagaraBool}
  User.LargeBody {type=NiagaraInt32}
 ]
}
